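# Post-process Animation Blueprint for the DoveSlide quadruped cosmetic companion, layering pose adjustments onto the incoming pose.
# The input pose runs through the DoveSlide pose asset's curve-driven blend, then blends toward a non-looping zero pose through the DisableTongue bone mask, weighted by DisableTongueAlpha.
# AdditivePoseAnim is then applied additively at CosmeticPoseAlpha; neither variable is declared here.
# The event graph is empty, so this asset sets none of these values itself.
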AnimBlueprint /CosmeticCompanions/Assets/Quadruped/DoveSlide/Meshes/DoveSlide_Quadruped_Post_AnimBP
target_skeleton=&/DoveSlide_Quadruped_Skeleton
graphs [
 AnimGraph {nodes [
   n1 = Layered blend per bone(
    BasePose=DoveSlide_Quadruped_PoseAsset!(SourcePose:Pose Link Structure=Input Pose()) -> (Pose:Pose Link Structure) {
     is=title AnimGraphNode_PoseBlendNode
     blendOption=Linear
     poseAsset=&/DoveSlide_Quadruped_PoseAsset
    }
    BlendPoses_0=Sequence Player 'DoveSlide_Pose_Zero'() {bLoopAnimation=false, sequence=/CosmeticCompanions/Assets/Quadruped/DoveSlide/Animation/Poses/DoveSlide_Pose_Zero}
    BlendWeights_0=Get DisableTongueAlpha() -> (DisableTongueAlpha:double)
   ) {
    blendMode=BlendMask
    curveBlendOption=DoNotOverride
    blendMasks [&DoveSlide_Quadruped_Skeleton:DisableTongue]
    layerSetup []
   }
   n2 = Get AdditivePoseAnim() -> (AdditivePoseAnim:Anim Sequence Object Reference)
   Output Pose(Result=Apply Additive(
     Base=@n1
     Additive=Sequence Player!(Sequence:Anim Sequence Base Object Reference=@n2) -> (Pose)
     Alpha=Get CosmeticPoseAlpha() -> (CosmeticPoseAlpha:float)
    ) {alphaBoolBlend {blendOption=Linear}})
  ]}
 EventGraph {links [], nodes []}
]
variables [DisableTongueAlpha {type=number}]
